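# Unlit, two-sided translucent Niagara mesh-particle material for the big reactor energy flame on the Nucleus glider.
# Three upward or sideways-panned noise samples modulate UV gradients, then CheapContrast hardens them into a flame mask.
# The mask blends two colors into HDR-adjusted emissive and, with a square gradient texture, feeds depth-faded opacity.
# SplineThicken supplies world position offset with separate base and tip widths.
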Material /BRCosmetics/Gadgets/Assets/VinderTech_GliderChute/Glider_Nucleus_Male/FX/M_Energy_Flame_Reactor_Big
graph {
 properties {
  bUsedWithNiagaraMeshParticles=true
  blendMode=BLEND_Translucent
  pixelDepthOffsetMode=PDOM_Legacy
  shadingModel=MSM_Unlit
  twoSided=true
  usedShadingModels=Unlit
 }
 nodes [
  n1 = ScalarParameter() {defaultValue=3.5, parameterName=WidthBase}
  n2 = ScalarParameter() {defaultValue=1, parameterName=WidthTip}
  Add(B=Add(B=MaterialFunctionCall(WidthBase=@n1, WidthTip=@n2) {materialFunction=&SplineThicken}.WorldPositionOffset))
  n3 = MaterialFunctionCall() {materialFunction=&LinearGradient}
  n4 = Power(Base=Multiply(A=@n3.UGradient, B=OneMinus(None=@n3.UGradient))) {constExponent=0.5}
  n5 = Multiply(A=VectorParameter() {group=FlameShape, parameterName=Flames_B_UVScale, defaultValue=<rgba 0.2 0.5 0 0>}.R, B=TextureCoordinate())
  n6 = Multiply(A=Time(), B=ScalarParameter() {defaultValue=0.5, group=FlameShape, parameterName=Flame_Speed})
  n7 = Multiply(
   A=ComponentMask(None=VectorParameter() {group=FlameShape, parameterName=Flames_M_UVScale, defaultValue=<rgba 1 0.5 0 0>}.RGB) {g=true, r=true}
   B=TextureCoordinate()
  )
  n8 = Multiply(A=Time(), B=ScalarParameter() {defaultValue=0.5, group=FlameShape, parameterName=Flame_Speed})
  n9 = Add(
   A=TextureSample(UVs=Panner(Coordinate=@n5, Time=@n6) {speedX=0.25, speedY=10}) {texture=&T_NoisePack_01}.G
   B=TextureSample(UVs=Panner(Coordinate=@n7, Time=@n8) {speedX=0.5, speedY=8}) {texture=&T_NoisePack_01}.B
  )
  n10 = ComponentMask(None=VectorParameter() {group=FlameShape, parameterName=Flames_S_UVScale, defaultValue=<rgba 1 0.1 0 0>}.RGB) {g=true, r=true}
  n11 = TextureCoordinate()
  n12 = Time()
  n13 = ScalarParameter() {defaultValue=0.5, group=FlameShape, parameterName=Flame_Speed}
  n14 = TextureSample(UVs=Panner(Coordinate=Multiply(A=@n10, B=@n11), Time=Multiply(A=@n12, B=@n13)) {speedX=0.25, speedY=6}) {texture=&T_NoisePack_01}
  n15 = Add(A=@n3.VGradient, B=Multiply(A=@n3.VGradient, B=Add(A=@n9, B=@n14.B)))
  n16 = Clamp(None=MaterialFunctionCall(In=Multiply(A=@n4, B=@n15), Contrast=ScalarParameter() {defaultValue=0.2, group=FlameShape, parameterName=HardEdgeness}) {materialFunction=&CheapContrast}.Result)
  Add(B=@n16)
  n17 = TextureSample() {texture=/Game/Effects/Fort_Effects/Textures/Masks/T_Square_Gradient_M}
  DepthFade(
   Opacity=Power(Base=Multiply(A=@n16, B=@n17.RGB)) {constExponent=0.5}
   FadeDistance=ScalarParameter() {defaultValue=5, parameterName=Depth Fade Distance}
  )
  n18 = LinearInterpolate(
   A=VectorParameter() {group=FlameShape, parameterName=Color_A, defaultValue=<rgba 0 0 1 1>}.RGB
   B=VectorParameter() {group=FlameShape, parameterName=Color_B, defaultValue=<rgba 0 0.5 1 1>}.RGB
   Alpha=MaterialFunctionCall(In=@n16, Contrast=ScalarParameter() {defaultValue=2, group=FlameShape, parameterName=HardEdgeness01}) {materialFunction=&CheapContrast}.Result
  )
  n19 = VectorParameter() {group=FlameShape, parameterName=HDR_Tint}
  MaterialFunctionCall(Emissive=MaterialFunctionCall(
    LDR Input=Clamp(None=@n18)
    HDR Tint=Multiply(A=@n18, B=@n19.RGB)
    HDR Intensity Multiplier=ScalarParameter() {defaultValue=1, group=FlameShape, parameterName=HDR_Intensity}
   ) {materialFunction=&DeriveHDRfromLDR}.RemappedRange) {materialFunction=&MF_FX_EmissiveAdjustments}
  Multiply(A=Multiply(A=VertexColor().R, B=Constant() {r=0.5}))
  Multiply(A=Multiply(A=VertexColor().G, B=Constant() {r=0.5}))
 ]
}
